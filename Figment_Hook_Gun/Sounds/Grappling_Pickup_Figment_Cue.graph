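# Sound cue for the Figment Hook Gun grappling pickup, reusing the dual pistols pickup sound.
# A single wave player feeds the cue output, with no randomization or modulation.
# It uses weapon pickup attenuation and the loot collect sound class.
# A concurrency override allows one instance at a time and prevents new plays while it is active.
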SoundCue /Figment_Hook_Gun/Sounds/Grappling_Pickup_Figment_Cue
root=WavePlayer_2
nodes [{id=WavePlayer_2, properties {soundWaveAssetPtr=/Game/Sounds/Fort_Weapons/pistol/Dual/DualPistols_Pickup_01}}]
properties {
 attenuationSettings=&Fort_Weapon_Pickup_Att
 bOverrideConcurrency=true
 maxDistance=3500
 soundClassObject=&Fort_Loot_Collect
 concurrencyOverrides {maxCount=1, resolutionRule=PreventNew}
}
